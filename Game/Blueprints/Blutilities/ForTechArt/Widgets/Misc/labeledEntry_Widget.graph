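# Reusable labeled input row for tech-art Blutility tools, showing a text, multiline text, checkbox, spin box or combo box.
# PreConstruct applies defaults and label layout, then selects the switcher child whose class matches currentType.
# Value setters unbind events, write, then rebind by type; handlers re-broadcast through the widget's dispatchers.
# The spin box unbinds OnValueChanged but binds OnValueCommitted, and its commit handler contains a leftover debug print.
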WidgetBlueprint /Game/Blueprints/Blutilities/ForTechArt/Widgets/Misc/labeledEntry_Widget
graphs [
 EventGraph {nodes [
   n1 = Event PreConstruct()
   > Parent: PreConstruct(IsDesignTime=@n1.IsDesignTime)
   > initFloat()
   > initBool()
   > initComboBox()
   > initText()
   > initType()
   > initLabel()
   Event Construct()
   > Parent: Construct()
   "On Value Committed (SpinBox_48)"!() -> (OutputDelegate, then, InValue:float, CommitMethod:ETextCommit Enum=Default) {is=title K2Node_ComponentBoundEvent}
   Event Tick()
  ]}
 clearBoundEvents {nodes [
   clearBoundEvents() {is=title K2Node_FunctionEntry}
   > Unbind all Events from On Check State Changed!(execute:Exec, self:Check Box Object Reference=Get CheckBox_43() -> (CheckBox_43:Check Box Object Reference)) -> (then:Exec) {is=title K2Node_ClearDelegate}
   > Unbind all Events from On Selection Changed!(
    execute:Exec
    self:"Combo Box (String) Object Reference"=Get ComboBoxString_127() -> (ComboBoxString_127:"Combo Box (String) Object Reference")
   ) -> (then:Exec) {is=title K2Node_ClearDelegate}
   > Unbind all Events from On Text Committed!(execute:Exec, self:Text Box Object Reference=Get EditableTextBox_26() -> (EditableTextBox_26:Text Box Object Reference)) -> (then:Exec) {is=title K2Node_ClearDelegate}
   > Unbind all Events from On Value Changed!(execute:Exec, self:Spin Box Object Reference=Get SpinBox_48() -> (SpinBox_48:Spin Box Object Reference)) -> (then:Exec) {is=title K2Node_ClearDelegate}
   > "Unbind all Events from OnTextCommitted (Multi-Line Text Box)"!(
    execute:Exec
    self:"Text Box (Multi- Line) Object Reference"=Get MultiLineEditableTextBox_54() -> (MultiLineEditableTextBox_54:"Text Box (Multi- Line) Object Reference")
   ) -> (then:Exec) {is=title K2Node_ClearDelegate}
  ]}
 clearFloatRange {nodes [
   clearFloatRange() {is=title K2Node_FunctionEntry}
   n1 = Get SpinBox_48()
   > ClearMaxSliderValue(self:Spin Box Object Reference=@n1)
   > ClearMaxValue(self:Spin Box Object Reference=@n1)
   > ClearMinSliderValue(self:Spin Box Object Reference=@n1)
   > ClearMinValue(self:Spin Box Object Reference=@n1)
  ]}
 getBool {nodes [
   getBool() {is=title K2Node_FunctionEntry}
   > Return Node!(execute, bool:bool=IsChecked(self=Get CheckBox_43()))
  ]}
 getFloat {nodes [
   getFloat() {is=title K2Node_FunctionEntry}
   > Return Node!(execute, float:double=GetValue!(self:Spin Box Object Reference=Get SpinBox_48()) -> (ReturnValue:float))
  ]}
 getSelectedComboBoxOption {nodes [
   getSelectedComboBoxOption() {is=title K2Node_FunctionEntry}
   > Return Node!(
    execute
    selectedOption:String=GetSelectedOption!(self:"Combo Box (String) Object Reference"=Get ComboBoxString_127()) -> (ReturnValue:String)
   )
  ]}
 getText {nodes [
   getText() {is=title K2Node_FunctionEntry}
   n1 = Get MultiLineEditableTextBox_54()
   n2 = Get EditableTextBox_26()
   n3 = Get currentType() -> (currentType:labeledEntryDetailType_Enum Enum=NewEnumerator0)
   > Return Node!(
    execute
    text:Text=Select(
     Option 0:Text="GetText (Text Box)"!(self:Text Box Object Reference=@n2) -> (ReturnValue:Text)
     Option 1:Text="GetText (Multi-Line Text Box)"!(self:"Text Box (Multi- Line) Object Reference"=@n1) -> (ReturnValue:Text)
     Index="Equal (Enum)"(A:labeledEntryDetailType_Enum Enum=@n3, B:labeledEntryDetailType_Enum Enum=NewEnumerator2)
    ) -> (ReturnValue:Text)
   )
  ]}
 handleCheckBox_CheckChanged {nodes [
   n1 = handleCheckBox_CheckChanged!() -> (then, bIsChecked:bool=~) {is=title K2Node_FunctionEntry}
   > Call On Checked State Changed!(execute:Exec, self:Labeled Entry Widget Object Reference, isChecked:bool=@n1) -> (then:Exec) {is=title K2Node_CallDelegate}
  ]}
 handleComboBox_SelectionChanged {nodes [
   n1 = handleComboBox_SelectionChanged!() -> (then, SelectedItem:String, SelectionType:ESelectInfo Enum) {is=title K2Node_FunctionEntry}
   > Call On Combo Box Selection Changed!(execute:Exec, self:Labeled Entry Widget Object Reference, SelectedItem:String=@n1, SelectionType:ESelectInfo Enum=@n1.SelectionType) -> (then:Exec) {is=title K2Node_CallDelegate}
  ]}
 handleSpinBox_Committed {nodes [
   n1 = handleSpinBox_Committed!() -> (then, InValue:float=~, CommitMethod:ETextCommit Enum=Default) {is=title K2Node_FunctionEntry}
   > PrintString!(execute, InString=gsagagsagsa, bPrintToScreen, bPrintToLog, TextColor, Duration, Key) -> (then)
   > Call On Float Value Committed!(execute:Exec, self:Labeled Entry Widget Object Reference, InValue:double=@n1, CommitMethod:ETextCommit Enum=@n1.CommitMethod) -> (then:Exec) {is=title K2Node_CallDelegate}
  ]}
 handleTextBox_Committed {nodes [
   n1 = handleTextBox_Committed!() -> (then, Text:"Text (by ref)", CommitMethod:ETextCommit Enum=Default) {is=title K2Node_FunctionEntry}
   > Call On Text Committed!(execute:Exec, self:Labeled Entry Widget Object Reference, text:Text=@n1, commitMethod:ETextCommit Enum=@n1.CommitMethod) -> (then:Exec) {is=title K2Node_CallDelegate}
  ]}
 initBool {nodes [
   initBool() {is=title K2Node_FunctionEntry}
   > setBool!(execute, self, bool:bool=Get defaultBool() -> (defaultBool:bool)) -> (then)
  ]}
 initComboBox {nodes [
   initComboBox() {is=title K2Node_FunctionEntry}
   > setComboOptions!(
    execute
    self
    options:Array of Strings=Get defaultComboBoxOptions() -> (defaultComboBoxOptions:Array of Strings)
    selectedOption:String=Get defaultSelectedComboBoxOption() -> (defaultSelectedComboBoxOption:String)
   ) -> (then)
  ]}
 initFloat {nodes [
   initFloat() {is=title K2Node_FunctionEntry}
   > setFloat!(execute, self, float:double=Get defaultFloat() -> (defaultFloat:double)) -> (then)
   > Branch(Condition="Not Equal (Float)"!(A:double=Get defaultFloatMin() -> (defaultFloatMin:double), B:double=Get defaultFloatMax() -> (defaultFloatMax:double)) -> (ReturnValue:bool) {is=title K2Node_CallFunction})
   > setFloatRange_Min!(execute, self, min:double=Get defaultFloatMin()) -> (then)
   > setFloatRange_Max!(execute, self, max:double=Get defaultFloatMax()) -> (then)
  ]}
 initLabel {nodes [
   initLabel() {is=title K2Node_FunctionEntry}
   > setLabel!(execute, self, label:String=Get label() -> (label:String)) -> (then)
   n1 = Get labelWidth() -> (labelWidth:double)
   > SetSize!(
    execute
    self:Horizontal Box Slot Object Reference=SlotAsHorizontalBoxSlot!(Widget:Widget Object Reference=Get SizeBox_0() -> (SizeBox_0:Size Box Object Reference)) -> (ReturnValue:Horizontal Box Slot Object Reference)
    InSize_Value:float=SelectFloat(A=1, B=Get labelWidth(), bPickA="float <= float"!(A:double=@n1, B:double=0) -> (ReturnValue:bool) {is=title K2Node_CallFunction})
    InSize_SizeRule:ESlateSizeRule Enum=Get labelSizeRule() -> (labelSizeRule:ESlateSizeRule Enum=Automatic)
   ) -> (then)
   > SetHorizontalAlignment(
    self:Size Box Slot Object Reference=SlotAsSizeBoxSlot!(Widget:Widget Object Reference=Get txtLabel() -> (txtLabel:Common Text Object Reference)) -> (ReturnValue:Size Box Slot Object Reference)
    InHorizontalAlignment=Get labelHorizontalAlignment() -> (labelHorizontalAlignment:EHorizontalAlignment Enum=HAlign_Fill)
   )
   > n2 = Switch on ESlateSizeRule!(execute, Selection:ESlateSizeRule Enum=Get labelSizeRule() -> (labelSizeRule=Automatic)) -> (Automatic:Exec, Fill:Exec) {is=title K2Node_SwitchEnum}
   > n3 = Branch(Condition="float > float"(A=Get labelWidth(), B=0) {is=title K2Node_CallFunction})
   > SetWidthOverride(self=Get SizeBox_0(), InWidthOverride=Get labelWidth())
   ClearWidthOverride(execute=@n3.else, self:Size Box Object Reference=Get SizeBox_0())
   ClearWidthOverride(execute=@n2.Fill, self=Get SizeBox_0())
  ]}
 initText {nodes [
   initText() {is=title K2Node_FunctionEntry}
   > setText!(execute, self, text:Text=Get defaultText() -> (defaultText:Text)) -> (then)
  ]}
 initType {nodes [
   initType() {is=title K2Node_FunctionEntry}
   > n1 = For Each Loop(Array:Array of Widget Object References=GetAllChildren(self=Get switcher() -> (switcher:Widget Switcher Object Reference))) -> (Array Element:Widget Object Reference)
   n2 = Get currentType() -> (currentType=NewEnumerator0)
   > Branch(Condition="Equal (Class)"!(
     A:Object Class Reference=Get Class(Object=@n1) -> (ReturnValue:Widget Class Reference)
     B:Object Class Reference=Select!(
      NewEnumerator0:Object Class Reference=&EditableTextBox
      NewEnumerator2:Object Class Reference=/Script/UMG.MultiLineEditableTextBox
      NewEnumerator1:Object Class Reference=/Script/UMG.CheckBox
      NewEnumerator4:Object Class Reference=/Script/UMG.SpinBox
      NewEnumerator3:Object Class Reference=/Script/UMG.ComboBoxString
      Index:labeledEntryDetailType_Enum Enum=@n2
     ) -> (ReturnValue:Object Class Reference)
    ) -> (ReturnValue:bool) {is=title K2Node_CallFunction})
   > SetActiveWidgetIndex(self=Get switcher(), Index=@n1.Array Index)
   > Return Node!(execute)
  ]}
 onCheckedStateChanged {nodes [onCheckedStateChanged!() -> (then, isChecked:bool=~) {is=title K2Node_FunctionEntry}]}
 onComboBoxSelectionChanged {nodes [onComboBoxSelectionChanged!() -> (then, SelectedItem:String, SelectionType:ESelectInfo Enum) {is=title K2Node_FunctionEntry}]}
 onFloatValueCommitted {nodes [onFloatValueCommitted!() -> (then, InValue:double=~, CommitMethod:ETextCommit Enum=Default) {is=title K2Node_FunctionEntry}]}
 onTextCommitted {nodes [onTextCommitted!() -> (then, text:Text, commitMethod:ETextCommit Enum=Default) {is=title K2Node_FunctionEntry}]}
 setBool {nodes [
   n1 = setBool!() -> (then, bool:bool=~) {is=title K2Node_FunctionEntry}
   > clearBoundEvents() {is=title K2Node_CallFunction}
   > SetIsChecked!(execute, self:Check Box Object Reference=Get CheckBox_43(), InIsChecked:bool=@n1) -> (then)
   > signUpForEventsByType()
  ]}
 setComboOptions {nodes [
   n1 = setComboOptions!() -> (then, options:Array of Strings, selectedOption:String) {is=title K2Node_FunctionEntry}
   > clearBoundEvents()
   > n2 = For Each Loop(Array:Array of Strings=@n1) -> (Array Element:String)
   n3 = Branch(execute=@n2.Completed, Condition="Equal Exactly (String)"(A=@n1.selectedOption))
   > n4 = ClearSelection(self:"Combo Box (String) Object Reference"=Get ComboBoxString_127())
   signUpForEventsByType(execute=[@n4.then, @n5.then])
   n5 = SetSelectedOption!(execute=@n3.else, self:"Combo Box (String) Object Reference"=Get ComboBoxString_127(), Option:String=@n1.selectedOption) -> (then)
   AddOption!(execute=@n2.LoopBody, self:"Combo Box (String) Object Reference"=Get ComboBoxString_127(), Option:String=@n2) -> (then)
  ]}
 setFloat {nodes [
   n1 = setFloat!() -> (then, float:double=~) {is=title K2Node_FunctionEntry}
   > clearBoundEvents()
   > SetValue!(execute, self:Spin Box Object Reference=Get SpinBox_48(), NewValue:float=@n1) -> (then)
   > signUpForEventsByType()
  ]}
 setFloatRange_Max {nodes [
   n1 = setFloatRange_Max!() -> (then, max:double=~) {is=title K2Node_FunctionEntry}
   n2 = Get SpinBox_48()
   > SetMaxSliderValue!(execute, self:Spin Box Object Reference=@n2, NewValue:float=@n1) -> (then)
   > SetMaxValue!(execute, self:Spin Box Object Reference=@n2, NewValue:float=@n1) -> (then)
   > setFloat(float="Min (Float)"(A:double=@n1, B:double=getFloat!() -> (float:double) {is=title K2Node_CallFunction}) -> (ReturnValue:double) {is=title K2Node_CommutativeAssociativeBinaryOperator}) {is=title K2Node_CallFunction}
  ]}
 setFloatRange_Min {nodes [
   n1 = setFloatRange_Min!() -> (then, min:double=~) {is=title K2Node_FunctionEntry}
   n2 = Get SpinBox_48()
   > SetMinValue!(execute, self:Spin Box Object Reference=@n2, NewValue:float=@n1) -> (then)
   > SetMinSliderValue!(execute, self:Spin Box Object Reference=@n2, NewValue:float=@n1) -> (then)
   > setFloat(float="Max (Float)"(A=@n1, B=getFloat()))
  ]}
 setLabel {nodes [
   n1 = setLabel!() -> (then, label:String) {is=title K2Node_FunctionEntry}
   > "SetText (Text)"(self=Get txtLabel(), InText="To Text (String)"(InString=@n1))
   > Set label(label:String=@n1) -> (Output_Get:String)
  ]}
 setSelectedComboOption {nodes [
   n1 = setSelectedComboOption!() -> (then, option:String) {is=title K2Node_FunctionEntry}
   > clearBoundEvents()
   > n2 = Branch(Condition="Equal Exactly (String)"(A=@n1))
   > n3 = ClearSelection(self=Get ComboBoxString_127())
   signUpForEventsByType(execute=[@n3.then, @n4.then])
   n4 = SetSelectedOption(execute=@n2.else, self=Get ComboBoxString_127(), Option=@n1)
  ]}
 setText {nodes [
   n1 = setText!() -> (then, text:Text) {is=title K2Node_FunctionEntry}
   > clearBoundEvents()
   > "SetText (Text Box)"!(execute, self:Text Box Object Reference=Get EditableTextBox_26(), InText:Text=@n1) -> (then)
   > "SetText (Multi-Line Text Box)"!(execute, self:"Text Box (Multi- Line) Object Reference"=Get MultiLineEditableTextBox_54(), InText:Text=@n1) -> (then)
   > signUpForEventsByType()
  ]}
 signUpForEventsByType {nodes [
   signUpForEventsByType() {is=title K2Node_FunctionEntry}
   > n1 = Switch on labeledEntryDetailType_Enum!(execute, Selection:labeledEntryDetailType_Enum Enum=Get currentType() -> (currentType=NewEnumerator0)) -> (NewEnumerator0:Exec, NewEnumerator2:Exec, NewEnumerator1:Exec, NewEnumerator4:Exec, NewEnumerator3:Exec) {is=title K2Node_SwitchEnum}
   n2 = Create Event()
   Bind Event to On Text Committed(execute=[@n1.NewEnumerator0, @n1.NewEnumerator2], self:Text Box Object Reference=Get EditableTextBox_26(), Delegate=@n2) {is=title K2Node_AddDelegate}
   > "Bind Event to OnTextCommitted (Multi-Line Text Box)"(self:"Text Box (Multi- Line) Object Reference"=Get MultiLineEditableTextBox_54(), Delegate=@n2) {is=title K2Node_AddDelegate}
   Bind Event to On Check State Changed(execute=@n1.NewEnumerator1, self:Check Box Object Reference=Get CheckBox_43(), Delegate=Create Event()) {is=title K2Node_AddDelegate}
   Bind Event to On Selection Changed(execute=@n1.NewEnumerator3, self:"Combo Box (String) Object Reference"=Get ComboBoxString_127(), Delegate=Create Event()) {is=title K2Node_AddDelegate}
   Bind Event to On Value Committed(execute=@n1.NewEnumerator4, self:Spin Box Object Reference=Get SpinBox_48(), Delegate=Create Event()) {is=title K2Node_AddDelegate}
  ]}
]
variables [
 currentType {type=labeledEntryDetailType_Enum, default=bool}
 defaultBool {type=boolean}
 defaultComboBoxOptions {type="array<string>", default []}
 defaultFloat {type=number}
 defaultFloatMax {type=number}
 defaultFloatMin {type=number}
 defaultSelectedComboBoxOption {type=string, default=""}
 defaultText {type=string, default=""}
 label {type=string, default=Label}
 labelHorizontalAlignment {type=EHorizontalAlignment, default=HAlign_Right}
 labelSizeRule {type=ESlateSizeRule, default=Automatic}
 labelWidth {type=number}
 onCheckedStateChanged {}
 onComboBoxSelectionChanged {}
 onFloatValueCommitted {}
 onTextCommitted {}
]
widgets {
 info {inheritedWidgetCount=0, parentClass=&UserWidget, rootWidgetClass=&VerticalBox, widgetCount=14}
 widgets [
  VerticalBox_0 {widgetClassPath=&VerticalBox, widget=VerticalBox_0}
  Image_318 {
   widgetClassPath=&Image
   bIsVariable=true
   parent=VerticalBox_0
   widget=Image_318
   properties {
    brush {
     bIsDynamicallyLoaded=false
     drawAs=Image
     imageType=NoImage
     mirroring=NoMirror
     resourceName=None
     resourceObject=null
     tiling=NoTile
     imageSize=<32 2>
     margin=<ltrb 0 0 0 0>
     outlineSettings {
      bUseBrushTransparency=false
      roundingType=HalfHeightRadius
      width=0
      color {colorUseRule=UseColor_Specified, specifiedColor=<rgba 0 0 0 0>}
      cornerRadii=<0 0 0 0>
     }
     tintColor {colorUseRule=UseColor_Specified, specifiedColor=<rgba 1 1 1 1>}
     uVRegion {bIsValid=false, max=<0 0>, min=<0 0>}
    }
    colorAndOpacity=<rgba 0.0885 0.0885 0.0885 1>
   }
   slot {class=VerticalBoxSlot}
  }
  Border_29 {
   widgetClassPath=&Border
   parent=VerticalBox_0
   widget=Border_29
   properties {brushColor=<rgba 0.0573 0.0573 0.0573 1>, padding=<ltrb 2 1 2 2>}
   slot {class=VerticalBoxSlot}
  }
  HorizontalBox_0 {
   widgetClassPath=&HorizontalBox
   parent=Border_29
   widget=HorizontalBox_0
   slot {class=BorderSlot, properties {padding=<ltrb 2 1 2 2>}}
  }
  SizeBox_0 {
   widgetClassPath=&SizeBox
   bIsVariable=true
   parent=HorizontalBox_0
   widget=SizeBox_0
   slot {class=HorizontalBoxSlot, properties {verticalAlignment=VAlign_Center}}
  }
  txtLabel {
   widgetClassPath=&CommonTextBlock
   bIsVariable=true
   parent=SizeBox_0
   widget=txtLabel
   properties {
    text=Select an Item to Create:
    font {
     bForceMonospaced=false
     bMaterialIsStencil=false
     fontMaterial=null
     fontObject=&NotoSans
     letterSpacing=0
     monospacedWidth=1
     size=12
     skewAmount=0
     typefaceFontName=NotoSans-Regular
     outlineSettings {
      bApplyOutlineToDropShadows=false
      bMiteredCorners=false
      bSeparateFillAlpha=false
      outlineMaterial=null
      outlineSize=0
      outlineColor=<rgba 0 0 0 1>
     }
    }
    shadowOffset=<0 0>
   }
   slot {class=SizeBoxSlot}
  }
  Spacer_25 {widgetClassPath=&Spacer, parent=HorizontalBox_0, widget=Spacer_25, properties {size=<10 1>}, slot {class=HorizontalBoxSlot}}
  switcher {
   widgetClassPath=&WidgetSwitcher
   bIsVariable=true
   parent=HorizontalBox_0
   widget=switcher
   slot {class=HorizontalBoxSlot, properties {verticalAlignment=VAlign_Center, size {sizeRule=Fill, value=1}}}
  }
  MultiLineEditableTextBox_54 {
   widgetClassPath=&MultiLineEditableTextBox
   bIsVariable=true
   parent=switcher
   widget=MultiLineEditableTextBox_54
   properties {
    autoWrapText=false
    text=dsdsd
    widgetStyle {
     backgroundColor {colorUseRule=UseColor_Specified, specifiedColor=<rgba 0.8698 0.8698 0.8698 1>}
     backgroundImageFocused {
      bIsDynamicallyLoaded=false
      drawAs=Box
      imageType=FullColor
      mirroring=NoMirror
      resourceName=../../../Engine/Content/Slate/Common/TextBox_Hovered.png
      resourceObject=null
      tiling=NoTile
      imageSize=<0 0>
      margin=<ltrb 0.25 0.25 0.25 0.25>
      outlineSettings {
       bUseBrushTransparency=false
       roundingType=HalfHeightRadius
       width=0
       color {colorUseRule=UseColor_Specified, specifiedColor=<rgba 0 0 0 0>}
       cornerRadii=<0 0 0 0>
      }
      tintColor {colorUseRule=UseColor_Specified, specifiedColor=<rgba 1 1 1 1>}
      uVRegion {bIsValid=false, max=<0 0>, min=<0 0>}
     }
     backgroundImageHovered {
      bIsDynamicallyLoaded=false
      drawAs=Box
      imageType=FullColor
      mirroring=NoMirror
      resourceName=../../../Engine/Content/Slate/Common/TextBox_Hovered.png
      resourceObject=null
      tiling=NoTile
      imageSize=<0 0>
      margin=<ltrb 0.25 0.25 0.25 0.25>
      outlineSettings {
       bUseBrushTransparency=false
       roundingType=HalfHeightRadius
       width=0
       color {colorUseRule=UseColor_Specified, specifiedColor=<rgba 0 0 0 0>}
       cornerRadii=<0 0 0 0>
      }
      tintColor {colorUseRule=UseColor_Specified, specifiedColor=<rgba 1 1 1 1>}
      uVRegion {bIsValid=false, max=<0 0>, min=<0 0>}
     }
     backgroundImageNormal {
      bIsDynamicallyLoaded=false
      drawAs=Box
      imageType=FullColor
      mirroring=NoMirror
      resourceName=../../../Engine/Content/Slate/Common/TextBox.png
      resourceObject=null
      tiling=NoTile
      imageSize=<0 0>
      margin=<ltrb 0.25 0.25 0.25 0.25>
      outlineSettings {
       bUseBrushTransparency=false
       roundingType=HalfHeightRadius
       width=0
       color {colorUseRule=UseColor_Specified, specifiedColor=<rgba 0 0 0 0>}
       cornerRadii=<0 0 0 0>
      }
      tintColor {colorUseRule=UseColor_Specified, specifiedColor=<rgba 1 1 1 1>}
      uVRegion {bIsValid=false, max=<0 0>, min=<0 0>}
     }
     backgroundImageReadOnly {
      bIsDynamicallyLoaded=false
      drawAs=Box
      imageType=FullColor
      mirroring=NoMirror
      resourceName=../../../Engine/Content/Slate/Common/TextBox_ReadOnly.png
      resourceObject=null
      tiling=NoTile
      imageSize=<0 0>
      margin=<ltrb 0.25 0.25 0.25 0.25>
      outlineSettings {
       bUseBrushTransparency=false
       roundingType=HalfHeightRadius
       width=0
       color {colorUseRule=UseColor_Specified, specifiedColor=<rgba 0 0 0 0>}
       cornerRadii=<0 0 0 0>
      }
      tintColor {colorUseRule=UseColor_Specified, specifiedColor=<rgba 1 1 1 1>}
      uVRegion {bIsValid=false, max=<0 0>, min=<0 0>}
     }
     focusedForegroundColor {colorUseRule=UseColor_Specified, specifiedColor=<rgba 0.03 0.03 0.03 1>}
     foregroundColor {colorUseRule=UseColor_Specified, specifiedColor=<rgba 0.03 0.03 0.03 1>}
     hScrollBarPadding=<ltrb 0 0 0 0>
     padding=<ltrb 4 2 4 2>
     readOnlyForegroundColor {colorUseRule=UseColor_Specified, specifiedColor=<rgba 0.1094 0.1094 0.1094 1>}
     scrollBarStyle {
      thickness=16
      draggedThumbImage {
       bIsDynamicallyLoaded=false
       drawAs=Box
       imageType=FullColor
       mirroring=NoMirror
       resourceName=../../../Engine/Content/Slate/Common/Scrollbar_Thumb.png
       resourceObject=null
       tiling=NoTile
       imageSize=<0 0>
       margin=<ltrb 0.25 0.25 0.25 0.25>
       outlineSettings {
        bUseBrushTransparency=false
        roundingType=HalfHeightRadius
        width=0
        color {colorUseRule=UseColor_Specified, specifiedColor=<rgba 0 0 0 0>}
        cornerRadii=<0 0 0 0>
       }
       tintColor {colorUseRule=UseColor_Specified, specifiedColor=<rgba 1 1 1 1>}
       uVRegion {bIsValid=false, max=<0 0>, min=<0 0>}
      }
      horizontalBackgroundImage {
       bIsDynamicallyLoaded=false
       drawAs=NoDrawType
       imageType=NoImage
       mirroring=NoMirror
       resourceName=None
       resourceObject=null
       tiling=NoTile
       imageSize=<0 0>
       margin=<ltrb 0 0 0 0>
       outlineSettings {
        bUseBrushTransparency=false
        roundingType=HalfHeightRadius
        width=0
        color {colorUseRule=UseColor_Specified, specifiedColor=<rgba 0 0 0 0>}
        cornerRadii=<0 0 0 0>
       }
       tintColor {colorUseRule=UseColor_Specified, specifiedColor=<rgba 1 1 1 1>}
       uVRegion {bIsValid=false, max=<0 0>, min=<0 0>}
      }
      horizontalBottomSlotImage {
       bIsDynamicallyLoaded=false
       drawAs=Image
       imageType=FullColor
       mirroring=NoMirror
       resourceName=../../../Engine/Content/Slate/Common/Scrollbar_Background_Horizontal.png
       resourceObject=null
       tiling=NoTile
       imageSize=<8 8>
       margin=<ltrb 0 0 0 0>
       outlineSettings {
        bUseBrushTransparency=false
        roundingType=HalfHeightRadius
        width=0
        color {colorUseRule=UseColor_Specified, specifiedColor=<rgba 0 0 0 0>}
        cornerRadii=<0 0 0 0>
       }
       tintColor {colorUseRule=UseColor_Specified, specifiedColor=<rgba 1 1 1 1>}
       uVRegion {bIsValid=false, max=<0 0>, min=<0 0>}
      }
      horizontalTopSlotImage {
       bIsDynamicallyLoaded=false
       drawAs=Image
       imageType=FullColor
       mirroring=NoMirror
       resourceName=../../../Engine/Content/Slate/Common/Scrollbar_Background_Horizontal.png
       resourceObject=null
       tiling=NoTile
       imageSize=<8 8>
       margin=<ltrb 0 0 0 0>
       outlineSettings {
        bUseBrushTransparency=false
        roundingType=HalfHeightRadius
        width=0
        color {colorUseRule=UseColor_Specified, specifiedColor=<rgba 0 0 0 0>}
        cornerRadii=<0 0 0 0>
       }
       tintColor {colorUseRule=UseColor_Specified, specifiedColor=<rgba 1 1 1 1>}
       uVRegion {bIsValid=false, max=<0 0>, min=<0 0>}
      }
      hoveredThumbImage {
       bIsDynamicallyLoaded=false
       drawAs=Box
       imageType=FullColor
       mirroring=NoMirror
       resourceName=../../../Engine/Content/Slate/Common/Scrollbar_Thumb.png
       resourceObject=null
       tiling=NoTile
       imageSize=<0 0>
       margin=<ltrb 0.25 0.25 0.25 0.25>
       outlineSettings {
        bUseBrushTransparency=false
        roundingType=HalfHeightRadius
        width=0
        color {colorUseRule=UseColor_Specified, specifiedColor=<rgba 0 0 0 0>}
        cornerRadii=<0 0 0 0>
       }
       tintColor {colorUseRule=UseColor_Specified, specifiedColor=<rgba 1 1 1 1>}
       uVRegion {bIsValid=false, max=<0 0>, min=<0 0>}
      }
      normalThumbImage {
       bIsDynamicallyLoaded=false
       drawAs=Box
       imageType=FullColor
       mirroring=NoMirror
       resourceName=../../../Engine/Content/Slate/Common/Scrollbar_Thumb.png
       resourceObject=null
       tiling=NoTile
       imageSize=<0 0>
       margin=<ltrb 0.25 0.25 0.25 0.25>
       outlineSettings {
        bUseBrushTransparency=false
        roundingType=HalfHeightRadius
        width=0
        color {colorUseRule=UseColor_Specified, specifiedColor=<rgba 0 0 0 0>}
        cornerRadii=<0 0 0 0>
       }
       tintColor {colorUseRule=UseColor_Specified, specifiedColor=<rgba 1 1 1 1>}
       uVRegion {bIsValid=false, max=<0 0>, min=<0 0>}
      }
      verticalBackgroundImage {
       bIsDynamicallyLoaded=false
       drawAs=NoDrawType
       imageType=NoImage
       mirroring=NoMirror
       resourceName=None
       resourceObject=null
       tiling=NoTile
       imageSize=<0 0>
       margin=<ltrb 0 0 0 0>
       outlineSettings {
        bUseBrushTransparency=false
        roundingType=HalfHeightRadius
        width=0
        color {colorUseRule=UseColor_Specified, specifiedColor=<rgba 0 0 0 0>}
        cornerRadii=<0 0 0 0>
       }
       tintColor {colorUseRule=UseColor_Specified, specifiedColor=<rgba 1 1 1 1>}
       uVRegion {bIsValid=false, max=<0 0>, min=<0 0>}
      }
      verticalBottomSlotImage {
       bIsDynamicallyLoaded=false
       drawAs=Image
       imageType=FullColor
       mirroring=NoMirror
       resourceName=../../../Engine/Content/Slate/Common/Scrollbar_Background_Vertical.png
       resourceObject=null
       tiling=NoTile
       imageSize=<8 8>
       margin=<ltrb 0 0 0 0>
       outlineSettings {
        bUseBrushTransparency=false
        roundingType=HalfHeightRadius
        width=0
        color {colorUseRule=UseColor_Specified, specifiedColor=<rgba 0 0 0 0>}
        cornerRadii=<0 0 0 0>
       }
       tintColor {colorUseRule=UseColor_Specified, specifiedColor=<rgba 1 1 1 1>}
       uVRegion {bIsValid=false, max=<0 0>, min=<0 0>}
      }
      verticalTopSlotImage {
       bIsDynamicallyLoaded=false
       drawAs=Image
       imageType=FullColor
       mirroring=NoMirror
       resourceName=../../../Engine/Content/Slate/Common/Scrollbar_Background_Vertical.png
       resourceObject=null
       tiling=NoTile
       imageSize=<8 8>
       margin=<ltrb 0 0 0 0>
       outlineSettings {
        bUseBrushTransparency=false
        roundingType=HalfHeightRadius
        width=0
        color {colorUseRule=UseColor_Specified, specifiedColor=<rgba 0 0 0 0>}
        cornerRadii=<0 0 0 0>
       }
       tintColor {colorUseRule=UseColor_Specified, specifiedColor=<rgba 1 1 1 1>}
       uVRegion {bIsValid=false, max=<0 0>, min=<0 0>}
      }
     }
     textStyle {
      overflowPolicy=Clip
      transformPolicy=None
      colorAndOpacity {colorUseRule=UseColor_Specified, specifiedColor=<rgba 0.03 0.03 0.03 1>}
      font {
       bForceMonospaced=false
       bMaterialIsStencil=false
       fontMaterial=null
       fontObject=null
       letterSpacing=0
       monospacedWidth=1
       size=14
       skewAmount=0
       typefaceFontName=Regular
       outlineSettings {
        bApplyOutlineToDropShadows=false
        bMiteredCorners=false
        bSeparateFillAlpha=false
        outlineMaterial=null
        outlineSize=0
        outlineColor=<rgba 0 0 0 1>
       }
      }
      highlightColor {colorUseRule=UseColor_Specified, specifiedColor=<rgba 0.02 0.3 0 1>}
      highlightShape {
       bIsDynamicallyLoaded=false
       drawAs=Box
       imageType=FullColor
       mirroring=NoMirror
       resourceName=../../../Engine/Content/Slate/Common/TextBlockHighlightShape.png
       resourceObject=null
       tiling=NoTile
       imageSize=<0 0>
       margin=<ltrb 0.375 0.375 0.375 0.375>
       outlineSettings {
        bUseBrushTransparency=false
        roundingType=HalfHeightRadius
        width=0
        color {colorUseRule=UseColor_Specified, specifiedColor=<rgba 0 0 0 0>}
        cornerRadii=<0 0 0 0>
       }
       tintColor {colorUseRule=UseColor_Specified, specifiedColor=<rgba 1 1 1 1>}
       uVRegion {bIsValid=false, max=<0 0>, min=<0 0>}
      }
      selectedBackgroundColor {colorUseRule=UseColor_Specified, specifiedColor=<rgba 0.0208 0.0208 0.0208 1>}
      shadowColorAndOpacity=<rgba 0 0 0 1>
      shadowOffset=<0 0>
      strikeBrush {
       bIsDynamicallyLoaded=false
       drawAs=Image
       imageType=NoImage
       mirroring=NoMirror
       resourceName=None
       resourceObject=null
       tiling=NoTile
       imageSize=<32 32>
       margin=<ltrb 0 0 0 0>
       outlineSettings {
        bUseBrushTransparency=false
        roundingType=HalfHeightRadius
        width=0
        color {colorUseRule=UseColor_Specified, specifiedColor=<rgba 0 0 0 0>}
        cornerRadii=<0 0 0 0>
       }
       tintColor {colorUseRule=UseColor_Specified, specifiedColor=<rgba 1 1 1 1>}
       uVRegion {bIsValid=false, max=<0 0>, min=<0 0>}
      }
      underlineBrush {
       bIsDynamicallyLoaded=false
       drawAs=Image
       imageType=NoImage
       mirroring=NoMirror
       resourceName=None
       resourceObject=null
       tiling=NoTile
       imageSize=<32 32>
       margin=<ltrb 0 0 0 0>
       outlineSettings {
        bUseBrushTransparency=false
        roundingType=HalfHeightRadius
        width=0
        color {colorUseRule=UseColor_Specified, specifiedColor=<rgba 0 0 0 0>}
        cornerRadii=<0 0 0 0>
       }
       tintColor {colorUseRule=UseColor_Specified, specifiedColor=<rgba 1 1 1 1>}
       uVRegion {bIsValid=false, max=<0 0>, min=<0 0>}
      }
     }
     vScrollBarPadding=<ltrb 0 0 0 0>
    }
   }
   slot {class=WidgetSwitcherSlot}
  }
  EditableTextBox_26 {
   widgetClassPath=&EditableTextBox
   bIsVariable=true
   parent=switcher
   widget=EditableTextBox_26
   properties {widgetStyle {
     backgroundColor {colorUseRule=UseColor_Foreground, specifiedColor=<rgba 1 1 1 1>}
     backgroundImageFocused {
      bIsDynamicallyLoaded=false
      drawAs=Box
      imageType=FullColor
      mirroring=NoMirror
      resourceName=../../../Engine/Content/Slate/Common/TextBox_Hovered.png
      resourceObject=null
      tiling=NoTile
      imageSize=<0 0>
      margin=<ltrb 0.25 0.25 0.25 0.25>
      outlineSettings {
       bUseBrushTransparency=false
       roundingType=HalfHeightRadius
       width=0
       color {colorUseRule=UseColor_Specified, specifiedColor=<rgba 0 0 0 0>}
       cornerRadii=<0 0 0 0>
      }
      tintColor {colorUseRule=UseColor_Specified, specifiedColor=<rgba 1 1 1 1>}
      uVRegion {bIsValid=false, max=<0 0>, min=<0 0>}
     }
     backgroundImageHovered {
      bIsDynamicallyLoaded=false
      drawAs=Box
      imageType=FullColor
      mirroring=NoMirror
      resourceName=../../../Engine/Content/Slate/Common/TextBox_Hovered.png
      resourceObject=null
      tiling=NoTile
      imageSize=<0 0>
      margin=<ltrb 0.25 0.25 0.25 0.25>
      outlineSettings {
       bUseBrushTransparency=false
       roundingType=HalfHeightRadius
       width=0
       color {colorUseRule=UseColor_Specified, specifiedColor=<rgba 0 0 0 0>}
       cornerRadii=<0 0 0 0>
      }
      tintColor {colorUseRule=UseColor_Specified, specifiedColor=<rgba 1 1 1 1>}
      uVRegion {bIsValid=false, max=<0 0>, min=<0 0>}
     }
     backgroundImageNormal {
      bIsDynamicallyLoaded=false
      drawAs=Box
      imageType=FullColor
      mirroring=NoMirror
      resourceName=../../../Engine/Content/Slate/Common/TextBox.png
      resourceObject=null
      tiling=NoTile
      imageSize=<0 0>
      margin=<ltrb 0.25 0.25 0.25 0.25>
      outlineSettings {
       bUseBrushTransparency=false
       roundingType=HalfHeightRadius
       width=0
       color {colorUseRule=UseColor_Specified, specifiedColor=<rgba 0 0 0 0>}
       cornerRadii=<0 0 0 0>
      }
      tintColor {colorUseRule=UseColor_Specified, specifiedColor=<rgba 1 1 1 1>}
      uVRegion {bIsValid=false, max=<0 0>, min=<0 0>}
     }
     backgroundImageReadOnly {
      bIsDynamicallyLoaded=false
      drawAs=Box
      imageType=FullColor
      mirroring=NoMirror
      resourceName=../../../Engine/Content/Slate/Common/TextBox_ReadOnly.png
      resourceObject=null
      tiling=NoTile
      imageSize=<0 0>
      margin=<ltrb 0.25 0.25 0.25 0.25>
      outlineSettings {
       bUseBrushTransparency=false
       roundingType=HalfHeightRadius
       width=0
       color {colorUseRule=UseColor_Specified, specifiedColor=<rgba 0 0 0 0>}
       cornerRadii=<0 0 0 0>
      }
      tintColor {colorUseRule=UseColor_Specified, specifiedColor=<rgba 1 1 1 1>}
      uVRegion {bIsValid=false, max=<0 0>, min=<0 0>}
     }
     focusedForegroundColor {colorUseRule=UseColor_Specified, specifiedColor=<rgba 0.03 0.03 0.03 1>}
     foregroundColor {colorUseRule=UseColor_Specified, specifiedColor=<rgba 0.03 0.03 0.03 1>}
     hScrollBarPadding=<ltrb 0 0 0 0>
     padding=<ltrb 4 2 4 2>
     readOnlyForegroundColor {colorUseRule=UseColor_Foreground, specifiedColor=<rgba 1 0 1 1>}
     scrollBarStyle {
      thickness=16
      draggedThumbImage {
       bIsDynamicallyLoaded=false
       drawAs=Box
       imageType=FullColor
       mirroring=NoMirror
       resourceName=../../../Engine/Content/Slate/Common/Scrollbar_Thumb.png
       resourceObject=null
       tiling=NoTile
       imageSize=<0 0>
       margin=<ltrb 0.25 0.25 0.25 0.25>
       outlineSettings {
        bUseBrushTransparency=false
        roundingType=HalfHeightRadius
        width=0
        color {colorUseRule=UseColor_Specified, specifiedColor=<rgba 0 0 0 0>}
        cornerRadii=<0 0 0 0>
       }
       tintColor {colorUseRule=UseColor_Specified, specifiedColor=<rgba 1 1 1 1>}
       uVRegion {bIsValid=false, max=<0 0>, min=<0 0>}
      }
      horizontalBackgroundImage {
       bIsDynamicallyLoaded=false
       drawAs=NoDrawType
       imageType=NoImage
       mirroring=NoMirror
       resourceName=None
       resourceObject=null
       tiling=NoTile
       imageSize=<0 0>
       margin=<ltrb 0 0 0 0>
       outlineSettings {
        bUseBrushTransparency=false
        roundingType=HalfHeightRadius
        width=0
        color {colorUseRule=UseColor_Specified, specifiedColor=<rgba 0 0 0 0>}
        cornerRadii=<0 0 0 0>
       }
       tintColor {colorUseRule=UseColor_Specified, specifiedColor=<rgba 1 1 1 1>}
       uVRegion {bIsValid=false, max=<0 0>, min=<0 0>}
      }
      horizontalBottomSlotImage {
       bIsDynamicallyLoaded=false
       drawAs=Image
       imageType=FullColor
       mirroring=NoMirror
       resourceName=../../../Engine/Content/Slate/Common/Scrollbar_Background_Horizontal.png
       resourceObject=null
       tiling=NoTile
       imageSize=<8 8>
       margin=<ltrb 0 0 0 0>
       outlineSettings {
        bUseBrushTransparency=false
        roundingType=HalfHeightRadius
        width=0
        color {colorUseRule=UseColor_Specified, specifiedColor=<rgba 0 0 0 0>}
        cornerRadii=<0 0 0 0>
       }
       tintColor {colorUseRule=UseColor_Specified, specifiedColor=<rgba 1 1 1 1>}
       uVRegion {bIsValid=false, max=<0 0>, min=<0 0>}
      }
      horizontalTopSlotImage {
       bIsDynamicallyLoaded=false
       drawAs=Image
       imageType=FullColor
       mirroring=NoMirror
       resourceName=../../../Engine/Content/Slate/Common/Scrollbar_Background_Horizontal.png
       resourceObject=null
       tiling=NoTile
       imageSize=<8 8>
       margin=<ltrb 0 0 0 0>
       outlineSettings {
        bUseBrushTransparency=false
        roundingType=HalfHeightRadius
        width=0
        color {colorUseRule=UseColor_Specified, specifiedColor=<rgba 0 0 0 0>}
        cornerRadii=<0 0 0 0>
       }
       tintColor {colorUseRule=UseColor_Specified, specifiedColor=<rgba 1 1 1 1>}
       uVRegion {bIsValid=false, max=<0 0>, min=<0 0>}
      }
      hoveredThumbImage {
       bIsDynamicallyLoaded=false
       drawAs=Box
       imageType=FullColor
       mirroring=NoMirror
       resourceName=../../../Engine/Content/Slate/Common/Scrollbar_Thumb.png
       resourceObject=null
       tiling=NoTile
       imageSize=<0 0>
       margin=<ltrb 0.25 0.25 0.25 0.25>
       outlineSettings {
        bUseBrushTransparency=false
        roundingType=HalfHeightRadius
        width=0
        color {colorUseRule=UseColor_Specified, specifiedColor=<rgba 0 0 0 0>}
        cornerRadii=<0 0 0 0>
       }
       tintColor {colorUseRule=UseColor_Specified, specifiedColor=<rgba 1 1 1 1>}
       uVRegion {bIsValid=false, max=<0 0>, min=<0 0>}
      }
      normalThumbImage {
       bIsDynamicallyLoaded=false
       drawAs=Box
       imageType=FullColor
       mirroring=NoMirror
       resourceName=../../../Engine/Content/Slate/Common/Scrollbar_Thumb.png
       resourceObject=null
       tiling=NoTile
       imageSize=<0 0>
       margin=<ltrb 0.25 0.25 0.25 0.25>
       outlineSettings {
        bUseBrushTransparency=false
        roundingType=HalfHeightRadius
        width=0
        color {colorUseRule=UseColor_Specified, specifiedColor=<rgba 0 0 0 0>}
        cornerRadii=<0 0 0 0>
       }
       tintColor {colorUseRule=UseColor_Specified, specifiedColor=<rgba 1 1 1 1>}
       uVRegion {bIsValid=false, max=<0 0>, min=<0 0>}
      }
      verticalBackgroundImage {
       bIsDynamicallyLoaded=false
       drawAs=NoDrawType
       imageType=NoImage
       mirroring=NoMirror
       resourceName=None
       resourceObject=null
       tiling=NoTile
       imageSize=<0 0>
       margin=<ltrb 0 0 0 0>
       outlineSettings {
        bUseBrushTransparency=false
        roundingType=HalfHeightRadius
        width=0
        color {colorUseRule=UseColor_Specified, specifiedColor=<rgba 0 0 0 0>}
        cornerRadii=<0 0 0 0>
       }
       tintColor {colorUseRule=UseColor_Specified, specifiedColor=<rgba 1 1 1 1>}
       uVRegion {bIsValid=false, max=<0 0>, min=<0 0>}
      }
      verticalBottomSlotImage {
       bIsDynamicallyLoaded=false
       drawAs=Image
       imageType=FullColor
       mirroring=NoMirror
       resourceName=../../../Engine/Content/Slate/Common/Scrollbar_Background_Vertical.png
       resourceObject=null
       tiling=NoTile
       imageSize=<8 8>
       margin=<ltrb 0 0 0 0>
       outlineSettings {
        bUseBrushTransparency=false
        roundingType=HalfHeightRadius
        width=0
        color {colorUseRule=UseColor_Specified, specifiedColor=<rgba 0 0 0 0>}
        cornerRadii=<0 0 0 0>
       }
       tintColor {colorUseRule=UseColor_Specified, specifiedColor=<rgba 1 1 1 1>}
       uVRegion {bIsValid=false, max=<0 0>, min=<0 0>}
      }
      verticalTopSlotImage {
       bIsDynamicallyLoaded=false
       drawAs=Image
       imageType=FullColor
       mirroring=NoMirror
       resourceName=../../../Engine/Content/Slate/Common/Scrollbar_Background_Vertical.png
       resourceObject=null
       tiling=NoTile
       imageSize=<8 8>
       margin=<ltrb 0 0 0 0>
       outlineSettings {
        bUseBrushTransparency=false
        roundingType=HalfHeightRadius
        width=0
        color {colorUseRule=UseColor_Specified, specifiedColor=<rgba 0 0 0 0>}
        cornerRadii=<0 0 0 0>
       }
       tintColor {colorUseRule=UseColor_Specified, specifiedColor=<rgba 1 1 1 1>}
       uVRegion {bIsValid=false, max=<0 0>, min=<0 0>}
      }
     }
     textStyle {
      overflowPolicy=Clip
      transformPolicy=None
      colorAndOpacity {colorUseRule=UseColor_Foreground, specifiedColor=<rgba 1 0 1 1>}
      font {
       bForceMonospaced=false
       bMaterialIsStencil=false
       fontMaterial=null
       fontObject=null
       letterSpacing=0
       monospacedWidth=1
       size=14
       skewAmount=0
       typefaceFontName=Regular
       outlineSettings {
        bApplyOutlineToDropShadows=false
        bMiteredCorners=false
        bSeparateFillAlpha=false
        outlineMaterial=null
        outlineSize=0
        outlineColor=<rgba 0 0 0 1>
       }
      }
      highlightColor {colorUseRule=UseColor_Specified, specifiedColor=<rgba 0.02 0.3 0 1>}
      highlightShape {
       bIsDynamicallyLoaded=false
       drawAs=Box
       imageType=FullColor
       mirroring=NoMirror
       resourceName=../../../Engine/Content/Slate/Common/TextBlockHighlightShape.png
       resourceObject=null
       tiling=NoTile
       imageSize=<0 0>
       margin=<ltrb 0.375 0.375 0.375 0.375>
       outlineSettings {
        bUseBrushTransparency=false
        roundingType=HalfHeightRadius
        width=0
        color {colorUseRule=UseColor_Specified, specifiedColor=<rgba 0 0 0 0>}
        cornerRadii=<0 0 0 0>
       }
       tintColor {colorUseRule=UseColor_Specified, specifiedColor=<rgba 1 1 1 1>}
       uVRegion {bIsValid=false, max=<0 0>, min=<0 0>}
      }
      selectedBackgroundColor {colorUseRule=UseColor_Foreground, specifiedColor=<rgba 1 0 1 1>}
      shadowColorAndOpacity=<rgba 0 0 0 1>
      shadowOffset=<0 0>
      strikeBrush {
       bIsDynamicallyLoaded=false
       drawAs=Image
       imageType=NoImage
       mirroring=NoMirror
       resourceName=None
       resourceObject=null
       tiling=NoTile
       imageSize=<32 32>
       margin=<ltrb 0 0 0 0>
       outlineSettings {
        bUseBrushTransparency=false
        roundingType=HalfHeightRadius
        width=0
        color {colorUseRule=UseColor_Specified, specifiedColor=<rgba 0 0 0 0>}
        cornerRadii=<0 0 0 0>
       }
       tintColor {colorUseRule=UseColor_Specified, specifiedColor=<rgba 1 1 1 1>}
       uVRegion {bIsValid=false, max=<0 0>, min=<0 0>}
      }
      underlineBrush {
       bIsDynamicallyLoaded=false
       drawAs=Image
       imageType=NoImage
       mirroring=NoMirror
       resourceName=None
       resourceObject=null
       tiling=NoTile
       imageSize=<32 32>
       margin=<ltrb 0 0 0 0>
       outlineSettings {
        bUseBrushTransparency=false
        roundingType=HalfHeightRadius
        width=0
        color {colorUseRule=UseColor_Specified, specifiedColor=<rgba 0 0 0 0>}
        cornerRadii=<0 0 0 0>
       }
       tintColor {colorUseRule=UseColor_Specified, specifiedColor=<rgba 1 1 1 1>}
       uVRegion {bIsValid=false, max=<0 0>, min=<0 0>}
      }
     }
     vScrollBarPadding=<ltrb 0 0 0 0>
    }}
   slot {class=WidgetSwitcherSlot}
  }
  CheckBox_43 {
   widgetClassPath=&CheckBox
   bIsVariable=true
   parent=switcher
   widget=CheckBox_43
   properties {widgetStyle {
     checkBoxType=CheckBox
     backgroundHoveredImage {
      bIsDynamicallyLoaded=false
      drawAs=NoDrawType
      imageType=NoImage
      mirroring=NoMirror
      resourceName=None
      resourceObject=null
      tiling=NoTile
      imageSize=<0 0>
      margin=<ltrb 0 0 0 0>
      outlineSettings {
       bUseBrushTransparency=false
       roundingType=HalfHeightRadius
       width=0
       color {colorUseRule=UseColor_Specified, specifiedColor=<rgba 0 0 0 0>}
       cornerRadii=<0 0 0 0>
      }
      tintColor {colorUseRule=UseColor_Specified, specifiedColor=<rgba 1 1 1 1>}
      uVRegion {bIsValid=false, max=<0 0>, min=<0 0>}
     }
     backgroundImage {
      bIsDynamicallyLoaded=false
      drawAs=NoDrawType
      imageType=NoImage
      mirroring=NoMirror
      resourceName=None
      resourceObject=null
      tiling=NoTile
      imageSize=<0 0>
      margin=<ltrb 0 0 0 0>
      outlineSettings {
       bUseBrushTransparency=false
       roundingType=HalfHeightRadius
       width=0
       color {colorUseRule=UseColor_Specified, specifiedColor=<rgba 0 0 0 0>}
       cornerRadii=<0 0 0 0>
      }
      tintColor {colorUseRule=UseColor_Specified, specifiedColor=<rgba 1 1 1 1>}
      uVRegion {bIsValid=false, max=<0 0>, min=<0 0>}
     }
     backgroundPressedImage {
      bIsDynamicallyLoaded=false
      drawAs=NoDrawType
      imageType=NoImage
      mirroring=NoMirror
      resourceName=None
      resourceObject=null
      tiling=NoTile
      imageSize=<0 0>
      margin=<ltrb 0 0 0 0>
      outlineSettings {
       bUseBrushTransparency=false
       roundingType=HalfHeightRadius
       width=0
       color {colorUseRule=UseColor_Specified, specifiedColor=<rgba 0 0 0 0>}
       cornerRadii=<0 0 0 0>
      }
      tintColor {colorUseRule=UseColor_Specified, specifiedColor=<rgba 1 1 1 1>}
      uVRegion {bIsValid=false, max=<0 0>, min=<0 0>}
     }
     borderBackgroundColor {colorUseRule=UseColor_Specified, specifiedColor=<rgba 1 1 1 1>}
     checkedForeground {colorUseRule=UseColor_Foreground, specifiedColor=<rgba 1 0 1 1>}
     checkedHoveredForeground {colorUseRule=UseColor_Foreground, specifiedColor=<rgba 1 0 1 1>}
     checkedHoveredImage {
      bIsDynamicallyLoaded=false
      drawAs=Image
      imageType=FullColor
      mirroring=NoMirror
      resourceName=../../../Engine/Content/Slate/Common/CheckBox_Checked_Hovered.png
      resourceObject=null
      tiling=NoTile
      imageSize=<23 23>
      margin=<ltrb 0 0 0 0>
      outlineSettings {
       bUseBrushTransparency=false
       roundingType=HalfHeightRadius
       width=0
       color {colorUseRule=UseColor_Specified, specifiedColor=<rgba 0 0 0 0>}
       cornerRadii=<0 0 0 0>
      }
      tintColor {colorUseRule=UseColor_Specified, specifiedColor=<rgba 0.5 0.5 0.5 1>}
      uVRegion {bIsValid=false, max=<0 0>, min=<0 0>}
     }
     checkedImage {
      bIsDynamicallyLoaded=false
      drawAs=Image
      imageType=FullColor
      mirroring=NoMirror
      resourceName=../../../Engine/Content/Slate/Common/CheckBox_Checked_Hovered.png
      resourceObject=null
      tiling=NoTile
      imageSize=<23 23>
      margin=<ltrb 0 0 0 0>
      outlineSettings {
       bUseBrushTransparency=false
       roundingType=HalfHeightRadius
       width=0
       color {colorUseRule=UseColor_Specified, specifiedColor=<rgba 0 0 0 0>}
       cornerRadii=<0 0 0 0>
      }
      tintColor {colorUseRule=UseColor_Specified, specifiedColor=<rgba 1 1 1 1>}
      uVRegion {bIsValid=false, max=<0 0>, min=<0 0>}
     }
     checkedPressedForeground {colorUseRule=UseColor_Foreground, specifiedColor=<rgba 1 0 1 1>}
     checkedPressedImage {
      bIsDynamicallyLoaded=false
      drawAs=Image
      imageType=FullColor
      mirroring=NoMirror
      resourceName=../../../Engine/Content/Slate/Common/CheckBox_Checked.png
      resourceObject=null
      tiling=NoTile
      imageSize=<23 23>
      margin=<ltrb 0 0 0 0>
      outlineSettings {
       bUseBrushTransparency=false
       roundingType=HalfHeightRadius
       width=0
       color {colorUseRule=UseColor_Specified, specifiedColor=<rgba 0 0 0 0>}
       cornerRadii=<0 0 0 0>
      }
      tintColor {colorUseRule=UseColor_Specified, specifiedColor=<rgba 1 1 1 1>}
      uVRegion {bIsValid=false, max=<0 0>, min=<0 0>}
     }
     checkedSlateSound {resourceObject=null}
     foregroundColor {colorUseRule=UseColor_Foreground, specifiedColor=<rgba 1 0 1 1>}
     hoveredForeground {colorUseRule=UseColor_Foreground, specifiedColor=<rgba 1 0 1 1>}
     hoveredSlateSound {resourceObject=null}
     padding=<ltrb 2 0 0 0>
     pressedForeground {colorUseRule=UseColor_Foreground, specifiedColor=<rgba 1 0 1 1>}
     uncheckedHoveredImage {
      bIsDynamicallyLoaded=false
      drawAs=Image
      imageType=FullColor
      mirroring=NoMirror
      resourceName=../../../Engine/Content/Slate/Common/CheckBox.png
      resourceObject=null
      tiling=NoTile
      imageSize=<23 23>
      margin=<ltrb 0 0 0 0>
      outlineSettings {
       bUseBrushTransparency=false
       roundingType=HalfHeightRadius
       width=0
       color {colorUseRule=UseColor_Specified, specifiedColor=<rgba 0 0 0 0>}
       cornerRadii=<0 0 0 0>
      }
      tintColor {colorUseRule=UseColor_Specified, specifiedColor=<rgba 1 1 1 1>}
      uVRegion {bIsValid=false, max=<0 0>, min=<0 0>}
     }
     uncheckedImage {
      bIsDynamicallyLoaded=false
      drawAs=Image
      imageType=FullColor
      mirroring=NoMirror
      resourceName=../../../Engine/Content/Slate/Common/CheckBox.png
      resourceObject=null
      tiling=NoTile
      imageSize=<23 23>
      margin=<ltrb 0 0 0 0>
      outlineSettings {
       bUseBrushTransparency=false
       roundingType=HalfHeightRadius
       width=0
       color {colorUseRule=UseColor_Specified, specifiedColor=<rgba 0 0 0 0>}
       cornerRadii=<0 0 0 0>
      }
      tintColor {colorUseRule=UseColor_Specified, specifiedColor=<rgba 1 1 1 1>}
      uVRegion {bIsValid=false, max=<0 0>, min=<0 0>}
     }
     uncheckedPressedImage {
      bIsDynamicallyLoaded=false
      drawAs=Image
      imageType=FullColor
      mirroring=NoMirror
      resourceName=../../../Engine/Content/Slate/Common/CheckBox_Hovered.png
      resourceObject=null
      tiling=NoTile
      imageSize=<23 23>
      margin=<ltrb 0 0 0 0>
      outlineSettings {
       bUseBrushTransparency=false
       roundingType=HalfHeightRadius
       width=0
       color {colorUseRule=UseColor_Specified, specifiedColor=<rgba 0 0 0 0>}
       cornerRadii=<0 0 0 0>
      }
      tintColor {colorUseRule=UseColor_Specified, specifiedColor=<rgba 0.5 0.5 0.5 1>}
      uVRegion {bIsValid=false, max=<0 0>, min=<0 0>}
     }
     uncheckedSlateSound {resourceObject=null}
     undeterminedForeground {colorUseRule=UseColor_Foreground, specifiedColor=<rgba 1 0 1 1>}
     undeterminedHoveredImage {
      bIsDynamicallyLoaded=false
      drawAs=Image
      imageType=FullColor
      mirroring=NoMirror
      resourceName=../../../Engine/Content/Slate/Common/CheckBox_Undetermined_Hovered.png
      resourceObject=null
      tiling=NoTile
      imageSize=<23 23>
      margin=<ltrb 0 0 0 0>
      outlineSettings {
       bUseBrushTransparency=false
       roundingType=HalfHeightRadius
       width=0
       color {colorUseRule=UseColor_Specified, specifiedColor=<rgba 0 0 0 0>}
       cornerRadii=<0 0 0 0>
      }
      tintColor {colorUseRule=UseColor_Specified, specifiedColor=<rgba 1 1 1 1>}
      uVRegion {bIsValid=false, max=<0 0>, min=<0 0>}
     }
     undeterminedImage {
      bIsDynamicallyLoaded=false
      drawAs=Image
      imageType=FullColor
      mirroring=NoMirror
      resourceName=../../../Engine/Content/Slate/Common/CheckBox_Undetermined.png
      resourceObject=null
      tiling=NoTile
      imageSize=<23 23>
      margin=<ltrb 0 0 0 0>
      outlineSettings {
       bUseBrushTransparency=false
       roundingType=HalfHeightRadius
       width=0
       color {colorUseRule=UseColor_Specified, specifiedColor=<rgba 0 0 0 0>}
       cornerRadii=<0 0 0 0>
      }
      tintColor {colorUseRule=UseColor_Specified, specifiedColor=<rgba 1 1 1 1>}
      uVRegion {bIsValid=false, max=<0 0>, min=<0 0>}
     }
     undeterminedPressedImage {
      bIsDynamicallyLoaded=false
      drawAs=Image
      imageType=FullColor
      mirroring=NoMirror
      resourceName=../../../Engine/Content/Slate/Common/CheckBox_Undetermined_Hovered.png
      resourceObject=null
      tiling=NoTile
      imageSize=<23 23>
      margin=<ltrb 0 0 0 0>
      outlineSettings {
       bUseBrushTransparency=false
       roundingType=HalfHeightRadius
       width=0
       color {colorUseRule=UseColor_Specified, specifiedColor=<rgba 0 0 0 0>}
       cornerRadii=<0 0 0 0>
      }
      tintColor {colorUseRule=UseColor_Specified, specifiedColor=<rgba 0.5 0.5 0.5 1>}
      uVRegion {bIsValid=false, max=<0 0>, min=<0 0>}
     }
    }}
   slot {class=WidgetSwitcherSlot, properties {horizontalAlignment=HAlign_Left}}
  }
  SpinBox_48 {
   widgetClassPath=&SpinBox
   bIsVariable=true
   parent=switcher
   widget=SpinBox_48
   properties {
    foregroundColor {colorUseRule=UseColor_Specified, specifiedColor=<rgba 0.03 0.03 0.03 1>}
    widgetStyle {
     activeBackgroundBrush {
      bIsDynamicallyLoaded=false
      drawAs=NoDrawType
      imageType=NoImage
      mirroring=NoMirror
      resourceName=None
      resourceObject=null
      tiling=NoTile
      imageSize=<0 0>
      margin=<ltrb 0 0 0 0>
      outlineSettings {
       bUseBrushTransparency=false
       roundingType=HalfHeightRadius
       width=0
       color {colorUseRule=UseColor_Specified, specifiedColor=<rgba 0 0 0 0>}
       cornerRadii=<0 0 0 0>
      }
      tintColor {colorUseRule=UseColor_Foreground, specifiedColor=<rgba 1 1 1 1>}
      uVRegion {bIsValid=false, max=<0 0>, min=<0 0>}
     }
     activeFillBrush {
      bIsDynamicallyLoaded=false
      drawAs=Box
      imageType=FullColor
      mirroring=NoMirror
      resourceName=../../../Engine/Content/Slate/Common/Spinbox_Fill.png
      resourceObject=null
      tiling=NoTile
      imageSize=<0 0>
      margin=<ltrb 0.25 0.25 0.5 0.25>
      outlineSettings {
       bUseBrushTransparency=false
       roundingType=HalfHeightRadius
       width=0
       color {colorUseRule=UseColor_Specified, specifiedColor=<rgba 0 0 0 0>}
       cornerRadii=<0 0 0 0>
      }
      tintColor {colorUseRule=UseColor_Specified, specifiedColor=<rgba 1 1 1 1>}
      uVRegion {bIsValid=false, max=<0 0>, min=<0 0>}
     }
     arrowsImage {
      bIsDynamicallyLoaded=false
      drawAs=Image
      imageType=FullColor
      mirroring=NoMirror
      resourceName=../../../Engine/Content/Slate/Common/SpinArrows.png
      resourceObject=null
      tiling=NoTile
      imageSize=<12 12>
      margin=<ltrb 0 0 0 0>
      outlineSettings {
       bUseBrushTransparency=false
       roundingType=HalfHeightRadius
       width=0
       color {colorUseRule=UseColor_Specified, specifiedColor=<rgba 0 0 0 0>}
       cornerRadii=<0 0 0 0>
      }
      tintColor {colorUseRule=UseColor_Specified, specifiedColor=<rgba 1 1 1 1>}
      uVRegion {bIsValid=false, max=<0 0>, min=<0 0>}
     }
     backgroundBrush {
      bIsDynamicallyLoaded=false
      drawAs=Box
      imageType=FullColor
      mirroring=NoMirror
      resourceName=../../../Engine/Content/Slate/Common/Spinbox.png
      resourceObject=null
      tiling=NoTile
      imageSize=<0 0>
      margin=<ltrb 0.25 0.25 0.25 0.25>
      outlineSettings {
       bUseBrushTransparency=false
       roundingType=HalfHeightRadius
       width=0
       color {colorUseRule=UseColor_Specified, specifiedColor=<rgba 0 0 0 0>}
       cornerRadii=<0 0 0 0>
      }
      tintColor {colorUseRule=UseColor_Specified, specifiedColor=<rgba 1 1 1 1>}
      uVRegion {bIsValid=false, max=<0 0>, min=<0 0>}
     }
     foregroundColor {colorUseRule=UseColor_Specified, specifiedColor=<rgba 0.4729 0.4729 0.4729 1>}
     hoveredBackgroundBrush {
      bIsDynamicallyLoaded=false
      drawAs=Box
      imageType=FullColor
      mirroring=NoMirror
      resourceName=../../../Engine/Content/Slate/Common/Spinbox_Hovered.png
      resourceObject=null
      tiling=NoTile
      imageSize=<0 0>
      margin=<ltrb 0.25 0.25 0.25 0.25>
      outlineSettings {
       bUseBrushTransparency=false
       roundingType=HalfHeightRadius
       width=0
       color {colorUseRule=UseColor_Specified, specifiedColor=<rgba 0 0 0 0>}
       cornerRadii=<0 0 0 0>
      }
      tintColor {colorUseRule=UseColor_Specified, specifiedColor=<rgba 1 1 1 1>}
      uVRegion {bIsValid=false, max=<0 0>, min=<0 0>}
     }
     hoveredFillBrush {
      bIsDynamicallyLoaded=false
      drawAs=NoDrawType
      imageType=NoImage
      mirroring=NoMirror
      resourceName=None
      resourceObject=null
      tiling=NoTile
      imageSize=<0 0>
      margin=<ltrb 0 0 0 0>
      outlineSettings {
       bUseBrushTransparency=false
       roundingType=HalfHeightRadius
       width=0
       color {colorUseRule=UseColor_Specified, specifiedColor=<rgba 0 0 0 0>}
       cornerRadii=<0 0 0 0>
      }
      tintColor {colorUseRule=UseColor_Specified, specifiedColor=<rgba 1 1 1 1>}
      uVRegion {bIsValid=false, max=<0 0>, min=<0 0>}
     }
     inactiveFillBrush {
      bIsDynamicallyLoaded=false
      drawAs=Box
      imageType=FullColor
      mirroring=NoMirror
      resourceName=../../../Engine/Content/Slate/Common/Spinbox_Fill_Hovered.png
      resourceObject=null
      tiling=NoTile
      imageSize=<0 0>
      margin=<ltrb 0.25 0.25 0.25 0.25>
      outlineSettings {
       bUseBrushTransparency=false
       roundingType=HalfHeightRadius
       width=0
       color {colorUseRule=UseColor_Specified, specifiedColor=<rgba 0 0 0 0>}
       cornerRadii=<0 0 0 0>
      }
      tintColor {colorUseRule=UseColor_Specified, specifiedColor=<rgba 1 1 1 1>}
      uVRegion {bIsValid=false, max=<0 0>, min=<0 0>}
     }
     insetPadding=<ltrb 0 0 0 0>
     textPadding=<ltrb 1 2 1 2>
    }
   }
   slot {class=WidgetSwitcherSlot, properties {verticalAlignment=VAlign_Center}}
  }
  ComboBoxString_127 {
   widgetClassPath=&ComboBoxString
   bIsVariable=true
   parent=switcher
   widget=ComboBoxString_127
   properties {
    foregroundColor {colorUseRule=UseColor_Specified, specifiedColor=<rgba 0.03 0.03 0.03 1>}
    widgetStyle {
     comboButtonStyle {
      downArrowAlign=VAlign_Center
      buttonStyle {
       clickedSlateSound {resourceObject=null}
       disabled {
        bIsDynamicallyLoaded=false
        drawAs=NoDrawType
        imageType=NoImage
        mirroring=NoMirror
        resourceName=None
        resourceObject=null
        tiling=NoTile
        imageSize=<0 0>
        margin=<ltrb 0 0 0 0>
        outlineSettings {
         bUseBrushTransparency=false
         roundingType=HalfHeightRadius
         width=0
         color {colorUseRule=UseColor_Specified, specifiedColor=<rgba 0 0 0 0>}
         cornerRadii=<0 0 0 0>
        }
        tintColor {colorUseRule=UseColor_Specified, specifiedColor=<rgba 1 1 1 1>}
        uVRegion {bIsValid=false, max=<0 0>, min=<0 0>}
       }
       disabledForeground {colorUseRule=UseColor_Foreground, specifiedColor=<rgba 0.5271 0.5271 0.5271 1>}
       hovered {
        bIsDynamicallyLoaded=false
        drawAs=RoundedBox
        imageType=NoImage
        mirroring=NoMirror
        resourceName=None
        resourceObject=null
        tiling=NoTile
        imageSize=<32 32>
        margin=<ltrb 0 0 0 0>
        outlineSettings {
         bUseBrushTransparency=true
         roundingType=FixedRadius
         width=1
         color {colorUseRule=UseColor_Specified, specifiedColor=<rgba 0.7243 0.7243 0.7243 1>}
         cornerRadii=<4 4 4 4>
        }
        tintColor {colorUseRule=UseColor_Specified, specifiedColor=<rgba 0.7243 0.7243 0.7243 1>}
        uVRegion {bIsValid=false, max=<0 0>, min=<0 0>}
       }
       hoveredForeground {colorUseRule=UseColor_Foreground, specifiedColor=<rgba 0 1 1 1>}
       hoveredSlateSound {resourceObject=null}
       normal {
        bIsDynamicallyLoaded=false
        drawAs=RoundedBox
        imageType=NoImage
        mirroring=NoMirror
        resourceName=None
        resourceObject=null
        tiling=NoTile
        imageSize=<32 32>
        margin=<ltrb 0 0 0 0>
        outlineSettings {
         bUseBrushTransparency=true
         roundingType=FixedRadius
         width=1
         color {colorUseRule=UseColor_Specified, specifiedColor=<rgba 0.6951 0.6951 0.6951 1>}
         cornerRadii=<4 4 4 4>
        }
        tintColor {colorUseRule=UseColor_Specified, specifiedColor=<rgba 0.4955 0.4955 0.4955 1>}
        uVRegion {bIsValid=false, max=<0 0>, min=<0 0>}
       }
       normalForeground {colorUseRule=UseColor_Foreground, specifiedColor=<rgba 0 1 1 1>}
       normalPadding=<ltrb 12 1.5 12 1.5>
       pressed {
        bIsDynamicallyLoaded=false
        drawAs=RoundedBox
        imageType=NoImage
        mirroring=NoMirror
        resourceName=None
        resourceObject=null
        tiling=NoTile
        imageSize=<32 32>
        margin=<ltrb 0 0 0 0>
        outlineSettings {
         bUseBrushTransparency=true
         roundingType=FixedRadius
         width=1
         color {colorUseRule=UseColor_Specified, specifiedColor=<rgba 0.7243 0.7243 0.7243 1>}
         cornerRadii=<4 4 4 4>
        }
        tintColor {colorUseRule=UseColor_Specified, specifiedColor=<rgba 0.3843 0.3843 0.3843 1>}
        uVRegion {bIsValid=false, max=<0 0>, min=<0 0>}
       }
       pressedForeground {colorUseRule=UseColor_Foreground, specifiedColor=<rgba 0 1 1 1>}
       pressedPadding=<ltrb 12 1.5 12 1.5>
       pressedSlateSound {resourceObject=null}
      }
      contentPadding=<ltrb 5 5 5 5>
      downArrowImage {
       bIsDynamicallyLoaded=false
       drawAs=Image
       imageType=FullColor
       mirroring=NoMirror
       resourceName=../../../Engine/Content/Slate/Common/ComboArrow.png
       resourceObject=null
       tiling=NoTile
       imageSize=<8 8>
       margin=<ltrb 0 0 0 0>
       outlineSettings {
        bUseBrushTransparency=false
        roundingType=HalfHeightRadius
        width=0
        color {colorUseRule=UseColor_Specified, specifiedColor=<rgba 0 0 0 0>}
        cornerRadii=<0 0 0 0>
       }
       tintColor {colorUseRule=UseColor_Specified, specifiedColor=<rgba 1 1 1 1>}
       uVRegion {bIsValid=false, max=<0 0>, min=<0 0>}
      }
      downArrowPadding=<ltrb 2 2 2 2>
      menuBorderBrush {
       bIsDynamicallyLoaded=false
       drawAs=Box
       imageType=FullColor
       mirroring=NoMirror
       resourceName=../../../Engine/Content/Slate/Old/Menu_Background.png
       resourceObject=null
       tiling=NoTile
       imageSize=<0 0>
       margin=<ltrb 0.125 0.125 0.125 0.125>
       outlineSettings {
        bUseBrushTransparency=false
        roundingType=HalfHeightRadius
        width=0
        color {colorUseRule=UseColor_Specified, specifiedColor=<rgba 0 0 0 0>}
        cornerRadii=<0 0 0 0>
       }
       tintColor {colorUseRule=UseColor_Specified, specifiedColor=<rgba 1 1 1 1>}
       uVRegion {bIsValid=false, max=<0 0>, min=<0 0>}
      }
      menuBorderPadding=<ltrb 1 1 1 1>
      shadowColorAndOpacity=<rgba 0 0 0 1>
      shadowOffset=<0 0>
     }
     contentPadding=<ltrb 4 2 4 2>
     menuRowPadding=<ltrb 0 0 0 0>
     pressedSlateSound {resourceObject=null}
     selectionChangeSlateSound {resourceObject=null}
    }
   }
   slot {class=WidgetSwitcherSlot, properties {verticalAlignment=VAlign_Center}}
  }
  Image {
   widgetClassPath=&Image
   bIsVariable=true
   parent=VerticalBox_0
   widget=Image
   properties {
    brush {
     bIsDynamicallyLoaded=false
     drawAs=Image
     imageType=NoImage
     mirroring=NoMirror
     resourceName=None
     resourceObject=null
     tiling=NoTile
     imageSize=<32 2>
     margin=<ltrb 0 0 0 0>
     outlineSettings {
      bUseBrushTransparency=false
      roundingType=HalfHeightRadius
      width=0
      color {colorUseRule=UseColor_Specified, specifiedColor=<rgba 0 0 0 0>}
      cornerRadii=<0 0 0 0>
     }
     tintColor {colorUseRule=UseColor_Specified, specifiedColor=<rgba 1 1 1 1>}
     uVRegion {bIsValid=false, max=<0 0>, min=<0 0>}
    }
    colorAndOpacity=<rgba 0.0365 0.0365 0.0365 1>
   }
   slot {class=VerticalBoxSlot}
  }
 ]
}
